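; Anim notify state for the DirectedCut group emote that applies a camera modifier and can lock look input during its window.
; Begin optionally ignores look input and, for a locally controlled pawn, adds the hard-coded BP_BackToBack_CameraModifier with a configurable blend-in time.
; End re-enables input and removes the modifier via the CameraModifier variable, which Begin only validity-checks.
; Tick calls KeepCGPathOptimizations on the controller and prints the default targeting rotation as debug output.
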(fn Received_NotifyBegin
  (params (MeshComp :type Skeletal%20Mesh%20Component%20Object%20Reference) (Animation :type Anim%20Sequence%20Base%20Object%20Reference) (TotalDuration :type Float%20%28single-precision%29) (EventReference :type Anim%20Notify%20Event%20Reference%20Structure%20%28by%20ref%29))
  (Utilities|FlowControl|Sequence
    (:then_0
      (bind _asfort_pawn (Utilities|Casting|CastToFortPawn (Components|GetOwner MeshComp)))
      (bind _asfort_player_controller (Utilities|Casting|CastToFortPlayerController (Pawn|GetController _asfort_pawn)))
      (if (Variables|Default|GetbIgnoreLookInput)
        (Input|SetIgnoreLookInput _asfort_player_controller true)))
    (:then_1
      ;[0.02096 0.151 0.02562] ReceivedNotfiyBegin: Apply the camera modifier to the player.
      (if (Utilities|IsValidClass (Variables|Default|GetCameraModifier))
        (Utilities|IsValid MeshComp
          (:Is%20Valid
            (bind _asfort_player_pawn (Utilities|Casting|CastToFortPlayerPawn (Components|GetOwner MeshComp)))
            (if (Pawn|IsLocallyControlled _asfort_player_pawn)
              (bind _outmodifier (Class|WeaponCameraFunctionLibrary|FindorAddCameraModifierforPawnNative "/SparksCosmetics/Emotes/Group/DirectedCut/BP_BackToBack_CameraModifier.BP_BackToBack_CameraModifier_C" _asfort_player_pawn))
              (bind _asbp_back_to_back_camera_modifier (Utilities|Casting|CastToBP_BackToBack_CameraModifier _outmodifier))
              (Variables|/SparksCosmetics/Emotes/Group/DirectedCut/BP_BackToBack_CameraModifier.BP_BackToBack_CameraModifier_C|SetBlend%20In%20Time (Variables|Default|GetBlend%20in%20Time) _asbp_back_to_back_camera_modifier))))))
    (:then_2
      (return false))))

(fn Received_NotifyEnd
  (params (MeshComp :type Skeletal%20Mesh%20Component%20Object%20Reference) (Animation :type Anim%20Sequence%20Base%20Object%20Reference) (EventReference :type Anim%20Notify%20Event%20Reference%20Structure%20%28by%20ref%29))
  (Utilities|FlowControl|Sequence
    (:then_0
      (bind _asfort_pawn (Utilities|Casting|CastToFortPawn (Components|GetOwner MeshComp)))
      (bind _asfort_player_controller (Utilities|Casting|CastToFortPlayerController (Pawn|GetController _asfort_pawn)))
      (if (Variables|Default|GetbIgnoreLookInput)
        (Input|SetIgnoreLookInput _asfort_player_controller false)))
    (:then_1
      ;[0.09896 0.05799 0.01674] ReceivedNotifyEnd: Remove the camera modifier.
      (if (Utilities|IsValidClass (Variables|Default|GetCameraModifier))
        (Utilities|IsValid MeshComp
          (:Is%20Valid
            (bind _asfort_player_pawn (Utilities|Casting|CastToFortPlayerPawn (Components|GetOwner MeshComp)))
            (if (Pawn|IsLocallyControlled _asfort_player_pawn)
              (Utilities|IsValid (Fortnite|GetFortPlayerControllerfromActor _asfort_player_pawn)
                (:Is%20Valid
                  (bind _playercameramanager (Variables|/Script/Engine.PlayerController|GetPlayerCameraManager (Fortnite|GetFortPlayerControllerfromActor _asfort_player_pawn))
                    (:then
                      (bind _outmodifier (Class|WeaponCameraFunctionLibrary|FindorAddCameraModifierforPawnNative (Variables|Default|GetCameraModifier) _asfort_player_pawn))
                      (Game|Player|RemoveCameraModifier _playercameramanager _outmodifier))
                    (:else)))))))))
    (:then_2
      (return false))))

(fn Received_NotifyTick
  (params (MeshComp :type Skeletal%20Mesh%20Component%20Object%20Reference) (Animation :type Anim%20Sequence%20Base%20Object%20Reference) (FrameDeltaTime :type Float%20%28single-precision%29) (EventReference :type Anim%20Notify%20Event%20Reference%20Structure%20%28by%20ref%29))
  (bind _asfort_player_pawn (Utilities|Casting|CastToFortPlayerPawn (Components|GetOwner MeshComp))
    (:then
      (Utilities|IsValid (Fortnite|GetFortPlayerControllerfromActor _asfort_player_pawn)
        (:Is%20Valid
          (Fortnite|KeepCGPathOptimizations (Fortnite|GetFortPlayerControllerfromActor _asfort_player_pawn))
          (Development|PrintString 0 (Utilities|String|ToString(Rotator) (Targeting|GetDefaultTargetingRotation _asfort_player_pawn)) true true "(R=0.000000,G=0.660000,B=1.000000,A=1.000000)" 0.0)
          (return false))
        (:Is%20Not%20Valid
          _)))
    (:CastFailed
      _))
  (return false))
