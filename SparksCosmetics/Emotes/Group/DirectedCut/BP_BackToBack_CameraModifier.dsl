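; Camera modifier for the Back to Back group emote that blends the player camera from its current view to the controlled pawn's location.
; A do-once SetValues captures the starting camera transform and a target rotation from the mesh root plus 90 degrees yaw.
; Each update advances a clamped alpha by DeltaTime over Blend In Time, lerping location and rotation.
; ParticipantsAverageLocation averages leader and follower positions but is uncalled here; debug prints remain.
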(fn BlueprintModifyCamera
  (params (DeltaTime :type Float%20%28single-precision%29) (ViewLocation :type Vector :default "0, 0, 0") (ViewRotation :type Rotator :default "0, 0, 0") (FOV :type Float%20%28single-precision%29))
  (bind _cameraowner (Variables|CameraModifier|GetCameraOwner))
  (bind _returnvalue (Game|Player|GetOwningPlayerController _cameraowner))
  (bind _returnvalue_2 (Pawn|GetControlledPawn _returnvalue))
  (bind _asfort_player_pawn (Utilities|Casting|CastToFortPlayerPawn _returnvalue_2))
  (bind _returnvalue_3 (Transformation|GetActorLocation _asfort_player_pawn))
  (CallFunction|SetValues)
  (Variables|Default|SetLerp (Math|Float|Clamp(Float) (+ (Variables|Default|GetLerp) (/ DeltaTime (Variables|Default|GetBlend%20In%20Time)))))
  (Development|PrintString (Utilities|String|ToString(Vector) _returnvalue_3) true true "(R=0.000000,G=0.660000,B=1.000000,A=1.000000)" 0.0)
  (return (Math|Vector|Lerp(Vector) (Variables|Default|GetFrom) _returnvalue_3 (Variables|Default|GetLerp)) (Math|Rotator|Lerp(Rotator) (Variables|Default|GetFromRot) (Variables|Default|GetPawnInitialRotation) (Variables|Default|GetLerp) true) (Camera|GetFOVAngle (Variables|CameraModifier|GetCameraOwner))))

(fn ParticipantsAverageLocation
  (params (Leader :type Fort%20Player%20Pawn%20Object%20Reference) (Followers :type Array%20of%20Fort%20Player%20Pawn%20Object%20References))
  (for _array_element Followers
    (Development|PrintString (Utilities|GetDisplayName _array_element) true true "(R=1.000000,G=0.000000,B=0.754801,A=1.000000)" 0.0)
    (LocalVariables|SetAverageVector (+ (LocalVariables|GetAverageVector) (Transformation|GetActorLocation _array_element))))
  (bind _output_get (LocalVariables|SetAverageVector (+ (LocalVariables|GetAverageVector) (Transformation|GetActorLocation Leader))))
  (bind _output_get_2 (LocalVariables|SetAverageVector (/ _output_get (+ (Utilities|Array|Length Followers) 1))))
  (return _output_get_2))

(fn SetValues
  (bind _cameraowner (Variables|CameraModifier|GetCameraOwner))
  (bind _returnvalue (Game|Player|GetOwningPlayerController _cameraowner))
  (bind _returnvalue_2 (Pawn|GetControlledPawn _returnvalue))
  (bind _asfort_player_pawn (Utilities|Casting|CastToFortPlayerPawn _returnvalue_2))
  (bind _cameraowner_2 (Variables|CameraModifier|GetCameraOwner))
  (if (Variables|Default|GetDoOnce)
    (Variables|Default|SetDoOnce false)
    (Variables|Default|SetFrom (Camera|GetCameraLocation _cameraowner_2))
    (Variables|Default|SetTo (Transformation|GetActorLocation _asfort_player_pawn))
    (Variables|Default|SetFromRot (Camera|GetCameraRotation _cameraowner_2))
    (Variables|Default|SetPawnInitialRotation (Math|Rotator|CombineRotators (Transformation|GetSocketRotation (Variables|/Script/Engine.Character|GetMesh _asfort_player_pawn) "root") "0.000000,90.000000,0.000000"))))
